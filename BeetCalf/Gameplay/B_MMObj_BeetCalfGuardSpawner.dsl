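; Radio tower guard spawner that registers with the BeetCalf tower game state component by spawner ID tag and unregisters on EndPlay.
; Guard count comes from a hotfixable per-spawner map keyed by playlist size, then selects an event table row of one elite plus up to two grunts.
; With radio tower determinism on, spawn requests run once, using the first boss patrol path and point; otherwise the parent logic handles them.
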(event EventBeginPlay
  (|CallParentFunction) ; This node is disabled and will not be called.\nDrag off pins to build functionality.
)

(event Collision|EventActorBeginOverlap
  (params (OtherActor :type Actor%20Object%20Reference)))

(event EventTick
  (params (DeltaSeconds :type Float%20%28single-precision%29)))

(event Initialize|EventInitializeSpawner
  (bind _returnvalue (Fortnite|Athena|GameState|GetGameStateAthena))
  (bind _returnvalue_2 (Actor|GetComponentbyClass _returnvalue "/BeetCalf/Gameplay/B_MMObj_BeetCalfTowerGameStateComponent.B_MMObj_BeetCalfTowerGameStateComponent_C"))
  ;[0.01864 0.1719 0.05333] InitializeSpawner: Register the radio tower guard spawner with the game state component.
  (|CallParentFunction)
  (Utilities|IsValid _returnvalue
    (:Is%20Valid
      (Utilities|IsValid _returnvalue_2
        (:Is%20Valid
          (Register|RegisterRadioTowerGuardSpawner _returnvalue_2 (Helpers|GetSpawnerIdTag) self :!class /BeetCalf/Gameplay/B_MMObj_BeetCalfTowerGameStateComponent.B_MMObj_BeetCalfTowerGameStateComponent_C))))))

(event EventEndPlay
  (params (EndPlayReason :type EEndPlayReason%20Enum :default "Destroyed"))
  (bind _returnvalue (Fortnite|Athena|GameState|GetGameStateAthena))
  (bind _returnvalue_2 (Actor|GetComponentbyClass _returnvalue "/BeetCalf/Gameplay/B_MMObj_BeetCalfTowerGameStateComponent.B_MMObj_BeetCalfTowerGameStateComponent_C"))
  ;[0.1458 0.0625 0.01494] EndPlay: Unregister the radio tower guard spawner from the game state component.
  (Utilities|FlowControl|Sequence
    (:then_0
      (Utilities|IsValid _returnvalue
        (:Is%20Valid
          (Utilities|IsValid _returnvalue_2
            (:Is%20Valid
              (Register|UnregisterRadioTowerGuardSpawner _returnvalue_2 (Helpers|GetSpawnerIdTag) self :!class /BeetCalf/Gameplay/B_MMObj_BeetCalfTowerGameStateComponent.B_MMObj_BeetCalfTowerGameStateComponent_C))))))
    (:then_1
      (|CallParentFunction EndPlayReason))))

(fn GetGuardCount
  (bind _spawneridtag (Helpers|GetSpawnerIdTag))
  (bind _hotfix_guardcountmap (Variables|Default|GetHotfix_GuardCountMap))
  (bind (_value _returnvalue) (Utilities|Map|Find _hotfix_guardcountmap _spawneridtag))
  (bind (_solo_2_faf6c8764d81d115c1214bbb3876b100 _duo_6_44834d0342181f5c5325958d89469e09 _trio_7_2becd65042c623029a1c83a56e8a7c69 _squad_8_85711d9b4d8b70275f495eb1eebf94bf) (Utilities|Struct|BreakMMObjPlaylistScalableFloatStruct _value))
  (bind _returnvalue_2 (Fortnite|Athena|GameState|GetGameStateAthena))
  (bind _playlistcontexttags (LocalVariables|GetPlaylistContextTags))
  ;[0.006821 0.07919 0.08333] GetGuardCount: Return the number of guards to spawn based on the playlist type.
  (Utilities|FlowControl|Sequence
    (:then_0
      (Utilities|IsValid _returnvalue_2
        (:Is%20Valid
          (LocalVariables|SetPlaylistContextTags (Class|FortGameStateAthena|GetAthenaPlaylistContextTags _returnvalue_2))
          (Utilities|FlowControl|Sequence
            (:then_0
              (if (GameplayTags|HasTag _playlistcontexttags "(TagName=\"Athena.Playlist.Solo\")")
                (LocalVariables|SetIndex 0)
                (else
                  (if (GameplayTags|HasTag _playlistcontexttags "(TagName=\"Athena.Playlist.Duo\")")
                    (LocalVariables|SetIndex 1)
                    (else
                      (if (GameplayTags|HasTag _playlistcontexttags "(TagName=\"Athena.Playlist.Trios\")")
                        (LocalVariables|SetIndex 2)
                        (else
                          (if (GameplayTags|HasTag _playlistcontexttags "(TagName=\"Athena.Playlist.Squad\")")
                            (LocalVariables|SetIndex 3)))))))))
            (:then_1
              (if _returnvalue
                (LocalVariables|SetRetCount (Utilities|Select :Index (Math|Integer|Clamp(Integer) (LocalVariables|GetIndex) 0 3) :Option%200 (ScalableFloat|GetValueasInteger _solo_2_faf6c8764d81d115c1214bbb3876b100) :Option%201 (ScalableFloat|GetValueasInteger _duo_6_44834d0342181f5c5325958d89469e09) :Option%202 (ScalableFloat|GetValueasInteger _trio_7_2becd65042c623029a1c83a56e8a7c69) :Option%203 (ScalableFloat|GetValueasInteger _squad_8_85711d9b4d8b70275f495eb1eebf94bf)))))))
        (:Is%20Not%20Valid
          ;;[1 0 0] Warning: GameStateAthena is null.
          (Fortnite|FortShippingLog self "GetGuardCount: GameStateAthena is null." true))))
    (:then_1
      (return (LocalVariables|GetRetCount)))))

(fn GetSpawnEventEntry
  ;[0.03084 0.1615 0.1356] Returns the spawn event entry to use based on the playlist type.
  (return (Utilities|Struct|MakeDataTableRowHandle (Variables|Default|GetEventTable) (Utilities|Select :Index (- (Math|Integer|Clamp(Integer) (Default|GetGuardCount) 1 3) 1) :Option%200 "RadioTowerGuards_1Elite" :Option%201 "RadioTowerGuards_1Elite1Grunt" :Option%202 "RadioTowerGuards_1Elite2Grunts"))))

(fn IsObjectiveEnabled
  ;[0.1927 0.04222 0.1578] Always return TRUE.
  ; IF the objective was not enabled, this actor wouldn't even have spawned.
  (return true))

(fn RequestSpawnGuard
  (params (RequestingActor :type Actor%20Object%20Reference))
  ;[0.01772 0.07833 0.1146] RequestSpawnGuard: Overridden to accomodate the deterministic setup.
  ; IF deterministic setup, spawn using the first patrol path on the first patrol point.
  (if (Class|FortDeterminismConfig|IsUsingRadioTowerDeterminism)
    (if (not (Variables|Transient|GetbRequestedSpawn))
      (Variables|Transient|SetbRequestedSpawn true)
      (Fortnite|FortShippingLog self "RequestSpawnGuard: Using deterministic setup.")
      (Variables|Transient|SetSpawnGuardRequestingActor RequestingActor)
      (bind _result (ArrayMacros|SafeArrayGet (Variables|Settings|GetBossPatrolPaths))
        (:%20%20
          (bind _output_get (Variables|Transient|SetBossChosenPatrolPath _result))
          (Utilities|IsValid _output_get
            (:Is%20Valid
              (bind _result_2 (ArrayMacros|SafeArrayGet (Class|FortAthenaPatrolPath|GetPatrolPoints _output_get))
                (:%20%20
                  (bind _output_get_2 (Variables|Transient|SetBossChosenPatrolPoint _result_2))
                  (Utilities|IsValid _output_get_2
                    (:Is%20Valid
                      (GuardSpawn|Protected|BeginGuardSpawnSequenceatPatrolPoint _output_get_2))))))))))
    (else
      (|CallParentFunction RequestingActor))))

(fn ConstructionScript
  (|CallParentFunction))
