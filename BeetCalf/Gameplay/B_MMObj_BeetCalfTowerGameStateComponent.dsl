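; Game state component that runs the BeetCalf radio tower match objective, from tower placement to guard spawning and analytics.
; At playlist ready, the server activates tower data layers from a datatable, chosen by chained EQS queries or in row order under determinism.
; Objective activation starts timed progress updates; a hotfixable storm phase plus delay completes it, advancing tower phase and spawning guards.
; Replicated flags drive progress broadcasts and nearest-terminal marking.
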(event EventBeginPlay
  ;[0.01502 0.175 0.1823] BeginPlay: Set up actor registry and enable the radio tower datalayers.
  (|CallParentFunction)
  (Networking|SwitchHasAuthority
    (:Authority
      (Variables|Default|SetRadioTowerPhaseState "NewEnumerator0")
      (Default|CallOnRadioTowerPhaseAdvance)
      (bind _gamestate (Class|FortAsyncActionCurrentPlaylistReady|PlaylistDataReadyAsync)
        (:OnReady
          (Fortnite|FortShippingLog self "BeginPlay: Invoked.")
          (General|LogDataLayers)
          (Activation|ActivateDataLayers)
          (Utilities|IsValid _gamestate
            (:Is%20Valid
              (bind _returnvalue (GameplayEventRouter|GetEventRouter "/Script/FortniteGame.FortGameState" _gamestate))
              (bind (_asynctaskproxy _payload) (GameplayEventRouter|StartListeningToStatefulEvent _returnvalue "/Script/FortniteGame.FortGamePhaseUpdatedEvent")
                (:then
                  (Variables|Default|SetGamePhaseUpdatedAsyncTask _asynctaskproxy))
                (:OnEventReceived
                  _)
                (:OnSavedEventStateExists
                  _))
              (Analytics|HandleGamePhaseChanged (Utilities|Struct|BreakFortGamePhaseUpdatedEvent _payload)))))))))

(event EventActivateObjective
  ;[0.04187 0.1979 0.08506] ActivateObjective: When the objective is activated, power up all the radio towers.
  (|CallParentFunction)
  (Networking|SwitchHasAuthority
    (:Authority
      (Utilities|FlowControl|DoOnce false
        (:Completed
          (Fortnite|FortShippingLog self "ActivateObjective: Invoked")
          ;;[0.4635 0.197 0.08937] All the radio towers should enter the Activating state.
          (Variables|Default|SetRadioTowerPhaseState "NewEnumerator6")
          (Default|CallOnRadioTowerPhaseAdvance)
          ;;[0.03732 0.1247 0.1771] Setup timers and variables to start updating the activation progress.
          (Variables|Transient|Activation|SetObjectiveActivateServerWorldTime (Fortnite|Sync|Time|GetServerWorldTimeSeconds))
          (bind _returnvalue (Utilities|Time|SetTimerbyEvent (EventDispatchers|CreateEvent :!event UpdateRadioTowerActivationProgress) (Math|Float|Max(Float) (ScalableFloat|GetValueatLevel (Variables|Hotfix|GetHotfix_UpdateRadioTowerActivationProgressInterval)) 1.0) true))
          (Variables|Transient|Activation|SetUpdateRadioTowerActivationProgressTimerHandle _returnvalue)
          (Activation|UpdateRadioTowerActivationProgress))))))

(event Custom|RadioTowerActivationComplete
  (bind _hotfix_guardspawninvokedelay (Variables|Hotfix|GetHotfix_GuardSpawnInvokeDelay))
  (bind _returnvalue (ScalableFloat|GetValueatLevel _hotfix_guardspawninvokedelay))
  ;[0.01285 0.1171 0.2604] RadioTowerActivationComplete: Once the activation progress is complete, let all the radio towers know that they should enter the Active state.
  (Utilities|SwitchHasAuthority
    (:Authority
      (Utilities|FlowControl|DoOnce false
        (:Completed
          (Variables|Transient|Activation|SetRadioTowerActivationProgress 1.0)
          (if (ScalableFloat|GetValueasBool (Variables|Hotfix|GetHotfix_MarkTerminalOnObjectiveActivate))
            (Variables|Transient|Activation|SetbRepNotifyMarkClosestTerminal true)
            (else
              _))
          (Utilities|Time|ClearandInvalidateTimerbyHandle (Variables|Transient|Activation|GetUpdateRadioTowerActivationProgressTimerHandle))
          (Utilities|Time|ClearandInvalidateTimerbyHandle (Variables|Transient|GetActivateCompleteDelayTimerHandle))
          (Variables|Default|SetRadioTowerPhaseState "NewEnumerator2")
          (Default|CallOnRadioTowerPhaseAdvance)
          (Utilities|FlowControl|Sequence
            (:then_1
              (if (> _returnvalue 0)
                (bind _returnvalue_2 (Utilities|Time|SetTimerbyEvent (EventDispatchers|CreateEvent :!event SpawnRadioTowerGuards) _returnvalue))
                (Variables|Transient|SetSpawnRadioTowerGuardTimerHandle _returnvalue_2)
                (else
                  (General|SpawnRadioTowerGuards))))))))))

(event EventEndPlay
  (params (EndPlayReason :type EEndPlayReason%20Enum :default "Destroyed"))
  ;[0.1667 0.08823 0.02458] EndPlay: Unbind delegates.
  (Utilities|FlowControl|Sequence
    (:then_0
      (Utilities|SwitchHasAuthority
        (:Authority
          (Utilities|FlowControl|Sequence
            (:then_0
              (bind _radiotowerdatalayerselection_eqsqueryinstance (Variables|Transient|GetRadioTowerDatalayerSelection_EQSQueryInstance)
                (:then
                  (AI|EQS|UnbindEventfromOnQueryFinished _radiotowerdatalayerselection_eqsqueryinstance (EventDispatchers|CreateEvent :!event HandleDatalayerSelectionQueryComplete)))
                (:else)))
            (:then_1
              (if (Editor|InEditororPIE)
                (Analytics|SubmitAnalytics)
                (else
                  _))
              (bind _gamephaseupdatedasynctask (Variables|Default|GetGamePhaseUpdatedAsyncTask)
                (:then
                  (GameplayEventRouter|StopListeningtoStatefulEvent _gamephaseupdatedasynctask))
                (:else)))))))
    (:then_1
      (|CallParentFunction EndPlayReason))))

(event Custom|DebugRadioTowerActivateObjective
  ;[0 0 0] Cheat ke * DebugRadioTowerActivateObjective
  ;
  ; Activate the radio tower objective.
  (Networking|SwitchHasAuthority
    (:Authority
      (CallFunction|ActivateObjective)
      (CallFunction|RadioTowerActivationComplete))))

(event Custom|DebugRadioTowerActivateDatalayerOnly
  (params (Input :type String))
  ;[0 0 0] Cheat ke * DebugRadioTowerActivateDatalayerOnly (names)
  ;
  ; Activates the desired radio tower datalayers and disable the others.
  (Networking|SwitchHasAuthority
    (:Authority
      (CallFunction|DebugRadioTowerDeactivateDatalayer "all")
      (CallFunction|DebugRadioTowerActivateDatalayer Input))))

(event Custom|DebugRadioTowerActivateDatalayer
  (params (Input :type String))
  ;[0 0 0] Cheat ke * DebugRadioTowerActivateDatalayer (names)
  ;
  ; Activates the desired radio tower datalayers.
  (Networking|SwitchHasAuthority
    (:Authority
      (bind _radiotowerspawninfodatatable (Variables|Settings|GetRadioTowerSpawnInfoDatatable)
        (:then
          (bind _outrownames (DataTable|GetDataTableRowNames _radiotowerspawninfodatatable))
          (for _array_element _outrownames
            (bind _returnvalue (Utilities|GetDataTableRow _radiotowerspawninfodatatable _array_element))
            (if (ScalableFloat|GetValueasBool (Utilities|Struct|BreakMMObjPOISpawnInfoStruct _returnvalue))
              (Utilities|IsValid (.DatalayerAsset_19_35CF832546D6BB81AB591C9AEB62F17D (Utilities|Struct|BreakMMObjPOISpawnInfoStruct _returnvalue))
                (:Is%20Valid
                  (if (or (Utilities|String|Equal,CaseInsensitive(String) (Utilities|String|ToLower Input) "all") (Utilities|Array|ContainsItem (Utilities|String|ParseIntoArray (Utilities|String|ToLower Input) ",") (Utilities|String|ToLower (Utilities|String|ToString(Name) _array_element))))
                    (bind _returnvalue_2 (WorldPartition|GetDataLayerManager))
                    (Utilities|IsValid _returnvalue_2
                      (:Is%20Valid
                        (DataLayers|SetDataLayerRuntimeState _returnvalue_2 (.DatalayerAsset_19_35CF832546D6BB81AB591C9AEB62F17D (Utilities|Struct|BreakMMObjPOISpawnInfoStruct _returnvalue)) "Activated" false :!class /Script/Engine.DataLayerManager)))))))))
        (:else)))))

(event Custom|DebugRadioTowerDeactivateDatalayer
  (params (Input :type String))
  ;[0 0 0] Cheat ke * DebugRadioTowerDeactivateDatalayer (names)
  ;
  ; Deactivates the desired radio tower datalayers.
  (Networking|SwitchHasAuthority
    (:Authority
      (bind _radiotowerspawninfodatatable (Variables|Settings|GetRadioTowerSpawnInfoDatatable)
        (:then
          (bind _outrownames (DataTable|GetDataTableRowNames _radiotowerspawninfodatatable))
          (for _array_element _outrownames
            (bind _returnvalue (Utilities|GetDataTableRow _radiotowerspawninfodatatable _array_element))
            (if (ScalableFloat|GetValueasBool (Utilities|Struct|BreakMMObjPOISpawnInfoStruct _returnvalue))
              (Utilities|IsValid (.DatalayerAsset_19_35CF832546D6BB81AB591C9AEB62F17D (Utilities|Struct|BreakMMObjPOISpawnInfoStruct _returnvalue))
                (:Is%20Valid
                  (if (or (Utilities|String|Equal,CaseInsensitive(String) (Utilities|String|ToLower Input) "all") (Utilities|Array|ContainsItem (Utilities|String|ParseIntoArray (Utilities|String|ToLower Input) ",") (Utilities|String|ToLower (Utilities|String|ToString(Name) _array_element))))
                    (bind _returnvalue_2 (WorldPartition|GetDataLayerManager))
                    (Utilities|IsValid _returnvalue_2
                      (:Is%20Valid
                        (DataLayers|SetDataLayerRuntimeState _returnvalue_2 (.DatalayerAsset_19_35CF832546D6BB81AB591C9AEB62F17D (Utilities|Struct|BreakMMObjPOISpawnInfoStruct _returnvalue)) "Unloaded" false :!class /Script/Engine.DataLayerManager)))))))))
        (:else)))))

(fn ActivateDataLayers
  ;[0.0614 0.1927 0.05798] ActivateDataLayers: Activate the radio tower datalayers in the world.
  (Networking|SwitchHasAuthority
    (:Authority
      (if (not (Variables|Transient|Activation|GetbDataLayersActivated))
        (Variables|Transient|Activation|SetbDataLayersActivated true)
        (Fortnite|FortShippingLog self "ActivateDataLayers: Invoked successfully.")
        (if (Class|FortDeterminismConfig|IsUsingRadioTowerDeterminism)
          (Fortnite|FortShippingLog self "ActivateDataLayers: Deterministic active.")
          (Activation|ActivateDeterministicDatalayers)
          (else
            (Activation|ActivateNextDataLayer)))
        (else
          ;;[1 0 0] Warning: Datalayers are already activated.
          (Fortnite|FortShippingLog self "ActivateDataLayers: DataLayers are already activated." true))))))

(fn ActivateDeterministicDatalayers
  ;[0.007986 0.1136 0.1562] ActivateDeterministicDatalayer: Activate the first found valid datalayers in the datatable.
  (Networking|SwitchHasAuthority
    (:Authority
      (bind _returnvalue (WorldPartition|GetDataLayerManager))
      (Utilities|IsValid _returnvalue
        (:Is%20Valid
          (bind _radiotowerspawninfodatatable (Variables|Settings|GetRadioTowerSpawnInfoDatatable)
            (:then
              (bind _outrownames (DataTable|GetDataTableRowNames _radiotowerspawninfodatatable))
              (for _array_element _outrownames
                (Utilities|FlowControl|Sequence
                  (:then_0
                    (if (>= (Utilities|Array|Length (Variables|Transient|Activation|GetActivatedDataLayers)) (CallFunction|GetNumRadioTowerstoActivate))
                      (return)))
                  (:then_1
                    (bind _returnvalue_2 (Utilities|GetDataTableRow _radiotowerspawninfodatatable _array_element))
                    (if (ScalableFloat|GetValueasBool (Utilities|Struct|BreakMMObjPOISpawnInfoStruct _returnvalue_2))
                      (Utilities|IsValid (.DatalayerAsset_19_35CF832546D6BB81AB591C9AEB62F17D (Utilities|Struct|BreakMMObjPOISpawnInfoStruct _returnvalue_2))
                        (:Is%20Valid
                          (bind _returnvalue_3 (DataLayers|SetDataLayerRuntimeState _returnvalue (.DatalayerAsset_19_35CF832546D6BB81AB591C9AEB62F17D (Utilities|Struct|BreakMMObjPOISpawnInfoStruct _returnvalue_2)) "Activated" false :!class /Script/Engine.DataLayerManager))
                          (if _returnvalue_3
                            (Utilities|Array|Add (Variables|Transient|Activation|GetActivatedDataLayers) _array_element)))))))))
            (:else)))))))

(fn ActivateNextDataLayer
  ;[0.05101 0.0144 0.05729] ActivateNextDataLayer: Run an EQS query to figure out the next radio tower datalayer to activate.
  (Networking|SwitchHasAuthority
    (:Authority
      (Variables|Transient|GetRadioTowerDatalayerSelection_EQSQueryInstance
        (:then
          ;;[1 0 0] Warning: EQSQuery is valid.
          (Fortnite|FortShippingLog self "ActivateNextDataLayer: EQSQuery Instance is already valid." true))
        (:else
          (if (< (Utilities|Array|Length (Variables|Transient|Activation|GetActivatedDataLayers)) (CallFunction|GetNumRadioTowerstoActivate))
            (bind _returnvalue (AI|EQS|RunEQSQuery "/BeetCalf/EQS/EQS_BeetCalfTowerDatalayerSelection.EQS_BeetCalfTowerDatalayerSelection" self))
            (Utilities|IsValid _returnvalue
              (:Is%20Valid
                (bind _output_get (Variables|Transient|SetRadioTowerDatalayerSelection_EQSQueryInstance _returnvalue))
                (AI|EQS|BindEventtoOnQueryFinished _output_get (EventDispatchers|CreateEvent :!event HandleDatalayerSelectionQueryComplete))
                (Fortnite|FortShippingLog self "ActivateNextDataLayer: EQS Query invoked successfully."))
              (:Is%20Not%20Valid
                ;;[1 0 0] Warning: Failed to create EQS Query.
                (Fortnite|FortShippingLog self "ActivateNextDataLayer: Failed to create EQS Query." true)))
            (else
              (Fortnite|FortShippingLog self "ActivateNextDataLayer: Ignoring request as desired number of radio tower datalayers have been activated."))))))))

(fn CompareGamePhaseState
  (params (A_SafeZonePhaseIndex :type Integer) (A_SafeZoneState :type EFortSafeZoneState%20Enum) (B_SafeZonePhaseIndex :type Integer) (B_SafeZoneState :type EFortSafeZoneState%20Enum))
  ;[0.04344 0.1719 0.02753] Compare phase index and state to figure out if A happens after B.
  (Math|Integer|CompareInt A_SafeZonePhaseIndex B_SafeZonePhaseIndex
    (:>
      (return true))
    (:==
      (return (> (Math|Conversions|ToInteger(Byte) A_SafeZoneState) (Math|Conversions|ToInteger(Byte) B_SafeZoneState))))
    (:<
      (return false))))

(fn GetNumRadioTowersToActivate
  (return (Math|Integer|SelectInt (Math|Integer|SelectInt (- (ScalableFloat|GetValueasInteger (Variables|Hotfix|GetHotfix_NumOfRadioTower)) 1) (ScalableFloat|GetValueasInteger (Variables|Hotfix|GetHotfix_NumOfRadioTower)) (Math|Random|RandomBool)) (ScalableFloat|GetValueasInteger (Variables|Hotfix|GetHotfix_NumOfRadioTower)) (and (> (ScalableFloat|GetValueasInteger (Variables|Hotfix|GetHotfix_NumOfRadioTower)) 1) (ScalableFloat|GetValueasBool (Variables|Hotfix|GetHotfix_NumOfRadioTower_RandomEnabled))))))

(fn GetRadioTowerActivationProgress
  (return (Variables|Transient|Activation|GetRadioTowerActivationProgress)))

(fn GetRadioTowerDatalayerLocations_Activated
  ;[0.02087 0.151 0.1345] Returns the locations where the datalayer has already been activated.
  ; Called from B_MMObj_EQC_ActivatedRadioTowerDatalayerLocations.
  (Utilities|FlowControl|Sequence
    (:then_0
      (for _array_element (Variables|Transient|Activation|GetActivatedDataLayers)
        (bind _radiotowerspawninfodatatable (Variables|Settings|GetRadioTowerSpawnInfoDatatable)
          (:then
            (bind _returnvalue (Utilities|GetDataTableRow _radiotowerspawninfodatatable _array_element))
            (if (ScalableFloat|GetValueasBool (Utilities|Struct|BreakMMObjPOISpawnInfoStruct _returnvalue))
              (Utilities|Array|Add (LocalVariables|GetRetLocations) (.SpawnLocation_15_5033779E420594BA62C2D5A783E10B93 (Utilities|Struct|BreakMMObjPOISpawnInfoStruct _returnvalue)))))
          (:else))))
    (:then_1
      (return (LocalVariables|GetRetLocations)))))

(fn GetRadioTowerDatalayerLocations_Pending
  ;[0.01686 0.09047 0.1302] Return the locations where radio tower datalayers can still be activated.
  ; Called from B_MMObj_EQS_RadioTowerDatalayerLocationGenerator.
  (Utilities|FlowControl|Sequence
    (:then_0
      (Utilities|SwitchHasAuthority
        (:Authority
          (bind _radiotowerspawninfodatatable (Variables|Settings|GetRadioTowerSpawnInfoDatatable)
            (:then
              (bind _outrownames (DataTable|GetDataTableRowNames _radiotowerspawninfodatatable))
              (for _array_element _outrownames
                (if (not (Utilities|Array|ContainsItem (Variables|Transient|Activation|GetActivatedDataLayers) _array_element))
                  (bind _returnvalue (Utilities|GetDataTableRow _radiotowerspawninfodatatable _array_element))
                  (if (ScalableFloat|GetValueasBool (Utilities|Struct|BreakMMObjPOISpawnInfoStruct _returnvalue))
                    (Utilities|Array|Add (LocalVariables|GetRetLocations) (.SpawnLocation_15_5033779E420594BA62C2D5A783E10B93 (Utilities|Struct|BreakMMObjPOISpawnInfoStruct _returnvalue)))))))
            (:else)))))
    (:then_1
      (return (LocalVariables|GetRetLocations)))))

(fn GetRadioTowerShouldSpawnGuards
  (return (Variables|Default|GetShouldSpawnRadioTowerGuards)))

(fn GetRadioTowerState
  (return (Variables|Default|GetRadioTowerPhaseState)))

(fn GetRadioTowerTerminals
  ;[0.02699 0.1512 0.1823] Return the radio tower terminals from the radio towers.
  (bind _outactors (Actor|GetAllActorsOfClass "/BeetCalf/Gameplay/BGA_MMObj_BeetCalfTerminal.BGA_MMObj_BeetCalfTerminal_C"))
  (return _outactors))

(fn HandleCapturedRadioTowerState
  (params (RadioTower :type B%20MMObj%20Beet%20Calf%20Tower%20Object%20Reference) (NewState :type MMObjRadioTowerState%20Enum :default "NewEnumerator5"))
  ;[0.05924 0.09937 0.1302] HandleRadioTowerStateChanged: Track all the radio towers that have been captured for analytics.
  (Networking|SwitchHasAuthority
    (:Authority
      (Utilities|IsValid RadioTower
        (:Is%20Valid
          (if (Utilities|Enum|Equal(Enum) NewState "NewEnumerator7")
            (Utilities|Array|AddUnique (Variables|Analytics|GetAnalytics_CapturedRadioTowers) (Utilities|GetDisplayName RadioTower))))))))

(fn HandleActivateCompleteDelay
  ;[0.009761 0.1379 0.1771] HandleActivateCompleteDelay: Called when the activate storm phase index and state are reached and the specified delay period has passed.
  (Utilities|SwitchHasAuthority
    (:Authority
      (CallFunction|RadioTowerActivationComplete))))

(fn HandleDatalayerSelectionQueryComplete
  (params (QueryInstance :type EQS%20Query%20Instance%20Object%20Reference) (QueryStatus :type EEnvQueryStatus%20Enum :default "Processing"))
  ;[0.2083 0.01803 0.1637] HandleDataLayerSelectionQueryComplete: Activate the datalayer found through the EQS query.
  (Networking|SwitchHasAuthority
    (:Authority
      (Fortnite|FortShippingLog self "HandleDataLayerSelectionQueryComplete: Invoked.")
      (Utilities|FlowControl|Sequence
        (:then_0
          (Utilities|IsValid QueryInstance
            (:Is%20Valid
              (bind _resultlocations (AI|EQS|GetQueryResultsasLocations QueryInstance))
              (if (Utilities|Array|IsValidIndex _resultlocations)
                (bind _radiotowerspawninfodatatable (Variables|Settings|GetRadioTowerSpawnInfoDatatable)
                  (:then
                    (bind _outrownames (DataTable|GetDataTableRowNames _radiotowerspawninfodatatable))
                    (bind _array_element (Utilities|Array|ForEachLoopwithBreak _outrownames)
                      (:LoopBody
                        (bind _returnvalue (Utilities|GetDataTableRow _radiotowerspawninfodatatable _array_element))
                        (if (and (ScalableFloat|GetValueasBool (Utilities|Struct|BreakMMObjPOISpawnInfoStruct _returnvalue)) (Utilities|Operators|Equal(==) (.SpawnLocation_15_5033779E420594BA62C2D5A783E10B93 (Utilities|Struct|BreakMMObjPOISpawnInfoStruct _returnvalue)) (Utilities|Array|Get(aref) _resultlocations) 1e+02))
                          (Utilities|IsValid (.DatalayerAsset_19_35CF832546D6BB81AB591C9AEB62F17D (Utilities|Struct|BreakMMObjPOISpawnInfoStruct _returnvalue))
                            (:Is%20Valid
                              (bind _returnvalue_2 (WorldPartition|GetDataLayerManager))
                              (Utilities|IsValid _returnvalue_2
                                (:Is%20Valid
                                  (bind _returnvalue_3 (DataLayers|SetDataLayerRuntimeState _returnvalue_2 (.DatalayerAsset_19_35CF832546D6BB81AB591C9AEB62F17D (Utilities|Struct|BreakMMObjPOISpawnInfoStruct _returnvalue)) "Activated" false :!class /Script/Engine.DataLayerManager))
                                  (if _returnvalue_3
                                    (Utilities|Array|Add (Variables|Transient|Activation|GetActivatedDataLayers) _array_element)
                                    (LocalVariables|SetbSuccessfullyActivatedDatalayer true)
                                    (Fortnite|FortShippingLog self (Utilities|String|Append "HandleDataLayerSelectionQueryComplete: Successfully activated DataLayer: " (Utilities|String|ToString(Name) _array_element))))
                                  ;;[1 0 0] Warning: Failed to set DataLayer to Activated.
                                  (Fortnite|FortShippingLog self (Utilities|String|Append "HandleDataLayerSelectionQueryComplete: Failed to set DataLayer to Activated: " (Utilities|String|ToString(Name) _array_element)) true))
                                (:Is%20Not%20Valid
                                  _))
                              ;;[1 0 0] Warning: DataLayerManager is null.
                              (Fortnite|FortShippingLog self "HandleDataLayerSelectionQueryComplete: DataLayerManager is null." true))
                            (:Is%20Not%20Valid
                              _))
                          ;;[1 0 0] Warning: DataLayer value in the datatable is null.
                          (Fortnite|FortShippingLog self (Utilities|String|Append "HandleDataLayerSelectionQueryComplete: DataLayerAsset is null for row: " (Utilities|String|ToString(Name) _array_element)) true)))))
                  (:else))
                (else
                  ;;[1 0 0] Warning: Query failed to find valid location.
                  (Fortnite|FortShippingLog self "HandleDataLayerSelectionQueryComplete: Query failed to find location." true))))
            (:Is%20Not%20Valid
              ;;[1 0 0] Warning: Query Instance is null.
              (Fortnite|FortShippingLog self "HandleDataLayerSelectionQueryComplete: Query instance is null." true))))
        (:then_1
          ;;[0.0365 0.08444 0.2656] Activate the next datalayer.
          (if (LocalVariables|GetbSuccessfullyActivatedDatalayer)
            (Variables|Transient|SetRadioTowerDatalayerSelection_EQSQueryInstance 0)
            (Activation|ActivateNextDataLayer)
            (else
              ;;[1 0 0] Warning: Failed to activate any datalayer.
              (Fortnite|FortShippingLog self "HandleDataLayerSelectionQueryComplete: Failed to activate any datalayer. Stopping further activation requests." true))))))))

(fn HandleGamePhaseChanged
  (params (GamePhase :type EAthenaGamePhase%20Enum))
  ;[0.0164 0.07935 0.1198] HandleGamePhaseChanged: Once we're in EndGame, submit the analytics.
  (Networking|SwitchHasAuthority
    (:Authority
      (if (Utilities|Enum|Equal(Enum) GamePhase "EndGame")
        (Analytics|SubmitAnalytics)
        (bind _gamephaseupdatedasynctask (Variables|Default|GetGamePhaseUpdatedAsyncTask)
          (:then
            (GameplayEventRouter|StopListeningtoStatefulEvent _gamephaseupdatedasynctask))
          (:else))))))

(fn HandleRadioTowerGuardsSpawned
  (params (Spawner :type B%20MMObj%20Boss%20Henchmen%20Rift%20Spawner%20Object%20Reference))
  (bind _returnvalue (Utilities|GetDisplayName Spawner))
  (bind _returnvalue_2 (GameplayTags|GetDebugStringfromGameplayTag))
  (bind _radiotowerguardspawnersmap (Variables|Transient|GetRadioTowerGuardSpawnersMap))
  (bind (_value _returnvalue_3) (Utilities|Map|Find _radiotowerguardspawnersmap (Helpers|GetSpawnerIdTag Spawner)))
  (bind _spawneridtag (Helpers|GetSpawnerIdTag Spawner))
  (bind _radiotowerguardspawnersmap_2 (Variables|Transient|GetRadioTowerGuardSpawnersMap))
  (bind _radiotowerguardspawnersmap_3 (Variables|Transient|GetRadioTowerGuardSpawnersMap))
  ;[0.0557 0.1194 0.1406] HandleRadioTowerGuardsSpawned: Delegate callback when the radio tower guards have spawned.
  ; Also check if all the radio tower guards have spawned.
  (Networking|SwitchHasAuthority
    (:Authority
      (Utilities|IsValid Spawner
        (:Is%20Valid
          (Default|UnbindEventfromOnGuardsSpawned Spawner (EventDispatchers|CreateEvent :!event HandleRadioTowerGuardsSpawned))
          (if _returnvalue_3
            (bind _structout (Utilities|Struct|SetmembersinMMObjBeetCalfGuardSpawnerInfoStruct _value true))
            (Utilities|Map|Add (Variables|Transient|GetRadioTowerGuardSpawnersMap) _spawneridtag _structout)
            (Fortnite|FortShippingLog self (Utilities|String|Append "HandleBunkerGuardsSpawned: BunkerGuards spawned for " _returnvalue " ; Bunker Tag: " _returnvalue_2))
            (bind _keys (Utilities|Map|Keys _radiotowerguardspawnersmap_2))
            (for _array_element _keys
              (if (not (.GuardsSpawned_5_FC148D6B4377A33140A4AF98BF81C8DE (Utilities|Struct|BreakMMObjBeetCalfGuardSpawnerInfoStruct (Utilities|Map|Find _radiotowerguardspawnersmap_2 _array_element))))
                (Fortnite|FortShippingLog self "HandleRadioTowerGuardsSpawned: Still waiting for all guards to spawn.")
                (return)))
            ;;[0.01929 0.1485 0.1562] Cache all the pawns that have spawned.
            (bind _keys_2 (Utilities|Map|Keys _radiotowerguardspawnersmap_3))
            (for _array_element_2 _keys_2
              (if (.ReturnValue (Utilities|Map|Find _radiotowerguardspawnersmap_3 _array_element_2))
                (Utilities|IsValid (Utilities|Struct|BreakMMObjBeetCalfGuardSpawnerInfoStruct (Utilities|Map|Find _radiotowerguardspawnersmap_3 _array_element_2))
                  (:Is%20Valid
                    (bind _guardpawns (Helpers|GetAllGuardPawns (Utilities|Struct|BreakMMObjBeetCalfGuardSpawnerInfoStruct (Utilities|Map|Find _radiotowerguardspawnersmap_3 _array_element_2))))
                    (Utilities|Array|AppendArray (Variables|Transient|GetGuardPawns) _guardpawns)))))
            (Guards|MarkClosestBossGuardforAllPlayers)
            (else
              ;;[1 0 0] Warning: RadioTowerGuardSpawner not in map.
              (Fortnite|FortShippingLog self (Utilities|String|Append "HandleRadioTowerGuardsSpawned: Cannot find radio tower guard spawner in the map: " _returnvalue " ; Bunker Tag: " _returnvalue_2) true))))))))

(fn HandleSafeZoneUpdated
  (params (PhaseIndex :type Integer) (SafeZoneState :type EFortSafeZoneState%20Enum))
  (bind _hotfix_activatecompletedelay_randomenabled (Variables|Hotfix|GetHotfix_ActivateCompleteDelay_RandomEnabled))
  (bind _returnvalue (ScalableFloat|GetValueasBool _hotfix_activatecompletedelay_randomenabled))
  (bind _returnvalue_2 (Math|Random|RandomBool))
  (bind _hotfix_activatecompletedelay_randomvariance (Variables|Hotfix|GetHotfix_ActivateCompleteDelay_RandomVariance))
  (bind _returnvalue_3 (ScalableFloat|GetValueatLevel _hotfix_activatecompletedelay_randomvariance))
  (bind _returnvalue_4 (Math|Random|RandomFloatinRange 0.0 _returnvalue_3))
  (bind _returnvalue_5 (- (ScalableFloat|GetValueatLevel (Variables|Hotfix|GetHotfix_ActivateCompleteDelay)) _returnvalue_4))
  (bind _returnvalue_6 (+ _returnvalue_4 (ScalableFloat|GetValueatLevel (Variables|Hotfix|GetHotfix_ActivateCompleteDelay))))
  (bind _returnvalue_7 (Math|Float|SelectFloat _returnvalue_6 _returnvalue_5 _returnvalue_2))
  (bind _hotfix_activatecompletedelay (Variables|Hotfix|GetHotfix_ActivateCompleteDelay))
  (bind _returnvalue_8 (ScalableFloat|GetValueatLevel _hotfix_activatecompletedelay))
  (bind _returnvalue_9 (Math|Float|SelectFloat _returnvalue_7 _returnvalue_8 _returnvalue))
  ;[0.006416 0.05939 0.0625] HandleSafeZoneUpdated: Overridden from parent to get a notification on whenever the safe zone phase or state is updated.
  (|CallParentFunction PhaseIndex SafeZoneState)
  (Utilities|SwitchHasAuthority
    (:Authority
      (bind _output_get (Variables|Default|SetActivateComepleteSafeZoneIndex (Math|Integer|SelectInt (Math|Integer|SelectInt (- (ScalableFloat|GetValueasInteger (Variables|Hotfix|GetHotfix_ActivateCompleteSafeZonePhaseIndex)) 1) (- (ScalableFloat|GetValueasInteger (Variables|Hotfix|GetHotfix_ActivateCompleteSafeZonePhaseIndex)) 1) (Math|Random|RandomBool)) (ScalableFloat|GetValueasInteger (Variables|Hotfix|GetHotfix_ActivateCompleteSafeZonePhaseIndex)) (and (ScalableFloat|GetValueasBool (Variables|Hotfix|GetHotfix_ActivateCompleteSafeZone_RandomEnabled)) (== (ScalableFloat|GetValueasInteger (Variables|Hotfix|GetHotfix_ActivateCompleteSafeZonePhaseIndex)) 2)))))
      (if (and (== _output_get PhaseIndex) (Utilities|Enum|Equal(Enum) (Utilities|Enum|BytetoEnumEFortSafeZoneState (Math|Conversions|ToByte(Integer) (ScalableFloat|GetValueasInteger (Variables|Hotfix|GetHotfix_ActivateCompleteSafeZoneState)))) SafeZoneState))
        (Variables|Transient|SetbActivateCompleteSafeZoneStateAndPhaseReached true)
        ;;[0.1094 0.02868 0.1086] IF any additional delay is specified, run the timer for that duration.
        (if (> _returnvalue_9 0)
          (bind _output_get_2 (Variables|Default|SetActivateCompleteDelay _returnvalue_9))
          (bind _returnvalue_10 (Utilities|Time|SetTimerbyEvent (EventDispatchers|CreateEvent :!event HandleActivateCompleteDelay) _output_get_2))
          (Variables|Transient|SetActivateCompleteDelayTimerHandle _returnvalue_10)
          (else
            (Activation|HandleActivateCompleteDelay)))))))

(fn IncrementGamePhaseState
  (params (SafeZonePhaseIndex :type Integer) (SafeZoneState :type EFortSafeZoneState%20Enum))
  ;[0.02715 0.08376 0.1562] Increment the safe zone phase/state.
  (switch Utilities|FlowControl|Switch|SwitchonEFortSafeZoneState SafeZoneState
    (:None
      _)
    (:Starting
      _)
    (:Holding
      (return SafeZonePhaseIndex "Shrinking"))
    (:Shrinking
      (return (+ SafeZonePhaseIndex 1) "Holding")))
  (return SafeZonePhaseIndex SafeZoneState))

(fn LogDataLayers
  ;[0.1042 0.04513 0.1667] LogDataLayers: Based on hotfix values, print out all the enable and disabled datalayers for debugging purposes.
  (Networking|SwitchHasAuthority
    (:Authority
      (bind _radiotowerspawninfodatatable (Variables|Settings|GetRadioTowerSpawnInfoDatatable)
        (:then
          (bind _outrownames (DataTable|GetDataTableRowNames _radiotowerspawninfodatatable))
          (for _array_element _outrownames
            (bind _returnvalue (Utilities|GetDataTableRow _radiotowerspawninfodatatable _array_element))
            (Utilities|IsValid (.DatalayerAsset_19_35CF832546D6BB81AB591C9AEB62F17D (Utilities|Struct|BreakMMObjPOISpawnInfoStruct _returnvalue))
              (:Is%20Valid
                (if (ScalableFloat|GetValueasBool (Utilities|Struct|BreakMMObjPOISpawnInfoStruct _returnvalue))
                  (Fortnite|FortShippingLog self (Utilities|String|Append "LogDataLayers: Enabled: " (Utilities|String|ToString(Name) _array_element)))
                  (else
                    (Fortnite|FortShippingLog self (Utilities|String|Append "LogDataLayers: Disabled: " (Utilities|String|ToString(Name) _array_element))))))
              (:Is%20Not%20Valid
                (Fortnite|FortShippingLog self (Utilities|String|Append "LogDataLayers: DataLayerAsset for Row " (Utilities|String|ToString(Name) _array_element) " is NULL.") true)))))
        (:else
          ;;[1 0.02106 0] Warning: RadioTowerSpawnInfoDatatable is null.
          (Fortnite|FortShippingLog self "LogDataLayers: RadioTowerSpawnInfoDatatable is null.." true))))))

(fn MarkClosestBossGuardForAllPlayers
  (bind _closestbosspawn (LocalVariables|GetClosestBossPawn))
  ;[0.0169 0.05615 0.06771] MarkClosestBossGuardForAllPlayers: Mark the closest boss pawn for all players.
  (Networking|SwitchHasAuthority
    (:Authority
      (Utilities|FlowControl|Sequence
        (:then_0
          (Fortnite|FortShippingLog self "MarkClosestBossGuardForAllPlayers: Invoked.")
          ;;[0.01929 0.1485 0.1562] Cache all the boss pawns that have spawned.
          (for _array_element (Variables|Transient|GetGuardPawns)
            (Utilities|IsValid _array_element
              (:Is%20Valid
                (if (GameplayTags|HasMatchingGameplayTag _array_element "(TagName=\"AISpawnerData.Type.MMObj.RadioTowerGuard.Boss\")")
                  (Utilities|Array|Add (LocalVariables|GetSpawnedBossPawns) _array_element))))))
        (:then_1
          (if (ScalableFloat|GetValueasBool (Variables|Hotfix|GetHotfix_MarkGuardOnObjectiveActivate))
            (LocalVariables|SetMarkOnObjectiveActivateDist2D (ScalableFloat|GetValueatLevel (Variables|Hotfix|GetHotfix_MarkGuardOnObjectiveActivateDistance2D)))
            ;;[0.02467 0.05179 0.1719] Mark the closest boss using the objective marker component.
            (for _array_element_2 (Fortnite|GetAllFortPlayerControllers true)
              (Utilities|IsValid _array_element_2
                (:Is%20Valid
                  (Utilities|IsValid (Pawn|GetViewTarget _array_element_2)
                    (:Is%20Valid
                      (bind _asfort_player_state_athena (Utilities|Casting|CastToFortPlayerStateAthena (Variables|/Script/Engine.Controller|GetPlayerState _array_element_2)))
                      (if (<= (Variables|/Script/FortniteGame.FortPlayerStateAthena|GetPlace _asfort_player_state_athena) 0)
                        (bind _output_get (LocalVariables|SetClosestBossPawn (Fortnite|GetClosestActorfromArray (Pawn|GetViewTarget _array_element_2) (LocalVariables|GetSpawnedBossPawns))))
                        (Utilities|IsValid _output_get
                          (:Is%20Valid
                            (if (not (<= (LocalVariables|GetMarkOnObjectiveActivateDist2D) 0))
                              (if (< (Math|Vector|Distance2DSquared(Vector) (Transformation|GetActorLocation _output_get) (Transformation|GetActorLocation (Pawn|GetViewTarget _array_element_2))) (Math|Float|Square (LocalVariables|GetMarkOnObjectiveActivateDist2D)))
                                _))
                            (Utilities|IsValid (Actor|GetComponentbyClass _array_element_2 "/Script/FortniteGame.AthenaMarkerComponent")
                              (:Is%20Valid
                                (bind _returnvalue (MarkerComponent|MarkActoronServer (Actor|GetComponentbyClass _array_element_2 "/Script/FortniteGame.AthenaMarkerComponent") _closestbosspawn false true))
                                (Fortnite|FortShippingLog self (Utilities|String|Append "MarkClosestBossGuardForAllPlayers: " (Utilities|String|Append (select (Utilities|IsValid (MarkerComponent|FindMarkerbyID (Actor|GetComponentbyClass _array_element_2 "/Script/FortniteGame.AthenaMarkerComponent") _returnvalue) :!class /Script/Engine.KismetSystemLibrary) " Successfully marked boss " " Failed to mark boss ") (Utilities|GetDisplayName _closestbosspawn)) " for player " (Fortnite|GetPlayerNameSafe _array_element_2)) (not (Utilities|IsValid (MarkerComponent|FindMarkerbyID (Actor|GetComponentbyClass _array_element_2 "/Script/FortniteGame.AthenaMarkerComponent") _returnvalue) :!class /Script/Engine.KismetSystemLibrary)))))))))))))))))))

(fn MarkClosestRadioTowerTerminalForAllPlayers
  ;[0.09896 0.01585 0.07957] Loop through all the players and mark the closest radio tower terminal for each player.
  (Networking|CanExecuteCosmeticEvents
    (:True
      (LocalVariables|SetMarkOnObjectiveActivateDist2D (ScalableFloat|GetValueatLevel (Variables|Hotfix|GetHotfix_MarkTerminalOnObjectiveActivateDistance)))
      (bind _terminals (General|GetRadioTowerTerminals))
      (if (Utilities|Array|IsNotEmpty _terminals)
        (for _array_element (Fortnite|GetLocalFortPlayerControllers true)
          (Utilities|IsValid _array_element
            (:Is%20Valid
              (Utilities|IsValid (Pawn|GetViewTarget _array_element)
                (:Is%20Valid
                  (bind _asfort_player_state_athena (Utilities|Casting|CastToFortPlayerStateAthena (Variables|/Script/Engine.Controller|GetPlayerState _array_element)))
                  (if (<= (Variables|/Script/FortniteGame.FortPlayerStateAthena|GetPlace _asfort_player_state_athena) 0)
                    (Utilities|IsValid (Actor|GetComponentbyClass _array_element "/Script/FortniteGame.AthenaMarkerComponent")
                      (:Is%20Valid
                        (bind _asbga_mmobj_beet_calf_terminal (Utilities|Casting|CastToBGA_MMObj_BeetCalfTerminal (Fortnite|GetClosestActorfromArray (Pawn|GetViewTarget _array_element) _terminals)))
                        (if (not (<= (LocalVariables|GetMarkOnObjectiveActivateDist2D) 0))
                          (if (<= (Math|Vector|Distance2DSquared(Vector) (Transformation|GetActorLocation _asbga_mmobj_beet_calf_terminal) (Transformation|GetActorLocation (Pawn|GetViewTarget _array_element))) (Math|Float|Square (LocalVariables|GetMarkOnObjectiveActivateDist2D)))
                            _))
                        (MarkableActor|SetCanBeMarked _asbga_mmobj_beet_calf_terminal true) ; Ensure the actor can be marked in case its state variable hasn't replicated yet.
                        (Fortnite|RegisterSoundIndicatorPlacementContext _array_element _asbga_mmobj_beet_calf_terminal "0, 0, 0" false)))))))))))))

(fn NotifyRadioTowerObjectiveComplete
  (params (PlayerPawn :type Fort%20Player%20Pawn%20Object%20Reference))
  ;[0.01101 0.07779 0.09375] NotifyRadioTowerObjectiveComplete: Used to mark this "objective" as complete for the input player.
  ; Called directly from BGA_MMObj_BeetCalfTerminal.
  (Networking|SwitchHasAuthority
    (:Authority
      (Utilities|IsValid PlayerPawn
        (:Is%20Valid
          (bind _asfort_player_state_athena (Utilities|Casting|CastToFortPlayerStateAthena (Variables|/Script/Engine.Pawn|GetPlayerState PlayerPawn)))
          (CallFunction|SetObjectiveCompletedStat (Class|FortPlayerStateAthena|GetSquadId _asfort_player_state_athena)))))))

(fn OnRep_bRepNotifyMarkClosestTerminal
  ;[0.1198 0.003449 0.1034] Mark the closest radio tower terminal for all players.
  (Networking|CanExecuteCosmeticEvents
    (:True
      (if (Variables|Transient|Activation|GetbRepNotifyMarkClosestTerminal)
        (General|MarkClosestRadioTowerTerminalforAllPlayers)))))

(fn OnRep_RadioTowerActivationProgress
  ;[0.2079 0.05571 0.2083] Broadcast the latest RadioTowerActivationProgress.
  (Default|CallOnRadioTowerActivationProgressUpdated (Variables|Transient|Activation|GetRadioTowerActivationProgress)))

(fn RegisterRadioTowerGuardSpawner
  (params (SpawnerIdTag :type Gameplay%20Tag%20Structure) (GuardSpawner :type B%20MMObj%20Beet%20Calf%20Guard%20Spawner%20Object%20Reference))
  (bind _radiotowerguardspawnersmap (Variables|Transient|GetRadioTowerGuardSpawnersMap))
  (bind (_value _returnvalue) (Utilities|Map|Find _radiotowerguardspawnersmap SpawnerIdTag))
  ;[0.02556 0.09908 0.1406] RegisterRadioTowerSpawner: Register the radio tower guard spawner and bind to know when its guards have spawned.
  (Networking|SwitchHasAuthority
    (:Authority
      (Utilities|IsValid GuardSpawner
        (:Is%20Valid
          (if (not _returnvalue)
            (Utilities|Map|Add _radiotowerguardspawnersmap SpawnerIdTag (Utilities|Struct|MakeMMObjBeetCalfGuardSpawnerInfoStruct GuardSpawner))
            (Default|BindEventtoOnGuardsSpawned GuardSpawner (EventDispatchers|CreateEvent :!event HandleRadioTowerGuardsSpawned))
            (Fortnite|FortShippingLog self (Utilities|String|Append "RegisterRadioTowerGuardSpawner: Registering bunker guard spawner: " (Utilities|GetDisplayName GuardSpawner)))
            (else
              ;;[1 0 0] Warning: Guard spawner is already registered.
              (Fortnite|FortShippingLog self (Utilities|String|Append "RegisterRadioTowerGuardSpawner: Spawner is already registered: " (Utilities|GetDisplayName GuardSpawner)) true))))))))

(fn SpawnRadioTowerGuards
  ;[0.01098 0.1667 0.1484] SpawnRadioTowerGuards: Spawn the guards at the radio tower.
  (Networking|SwitchHasAuthority
    (:Authority
      (Variables|Default|SetShouldSpawnRadioTowerGuards true)
      (Default|CallOnRadioTowerSpawnGuards))))

(fn SubmitAnalytics
  (bind _analytics_capturedradiotowers (Variables|Analytics|GetAnalytics_CapturedRadioTowers))
  (bind _returnvalue (Utilities|Array|Length _analytics_capturedradiotowers))
  (bind _analytics_capturedradiotowers_2 (Variables|Analytics|GetAnalytics_CapturedRadioTowers))
  (bind _returnvalue_2 (Utilities|String|JoinStringArray _analytics_capturedradiotowers_2 " ; "))
  ;[0.03754 0.005307 0.04167] SubmitAnalytics: Fire analytics if it hasn't been done yet.
  (Utilities|SwitchHasAuthority
    (:Authority
      (if (and (not (Replay|GetIsPlayingReplay)) (not (Variables|Analytics|GetbAnalyticsSubmitted)))
        (Variables|Analytics|SetbAnalyticsSubmitted true)
        (Utilities|FlowControl|Sequence
          (:then_0
            ;;[0.02283 0.151 0.08822] NumOfCapturedRadioTowers
            (bind _returnvalue_3 (Analytics|ConstructAnalyticsAttribute :AttributeName "NumOfCapturedRadioTowers" :IntValue _returnvalue :floatValue -1.0 :BoolValue false))
            (Utilities|Array|Add (LocalVariables|GetAnalyticsEventAttributes) _returnvalue_3)
            (Fortnite|FortShippingLog self (Utilities|String|Append "NumOfCapturedRadioTowers: " (Utilities|String|ToString(Integer) _returnvalue)))
            ;;[0.02424 0.08952 0.1823] CapturedRadioTowers
            (bind _returnvalue_4 (Analytics|ConstructAnalyticsAttribute "CapturedRadioTowers" _returnvalue_2))
            (Utilities|Array|Add (LocalVariables|GetAnalyticsEventAttributes) _returnvalue_4)
            (Fortnite|FortShippingLog self (Utilities|String|Append "CapturedRadioTowers: " _returnvalue_2)))
          (:then_1
            (Analytics|FireAnalyticsEventBlueprintArray "Core.MMObj.RadioTower" (LocalVariables|GetAnalyticsEventAttributes))))))))

(fn UnregisterRadioTowerGuardSpawner
  (params (SpawnerIdTag :type Gameplay%20Tag%20Structure) (GuardSpawner :type B%20MMObj%20Beet%20Calf%20Guard%20Spawner%20Object%20Reference))
  (bind _radiotowerguardspawnersmap (Variables|Transient|GetRadioTowerGuardSpawnersMap))
  (bind (_value _returnvalue) (Utilities|Map|Find _radiotowerguardspawnersmap SpawnerIdTag))
  ;[0.02556 0.09908 0.1406] UnregisterRadioTowerGuardSpawner: Unregister the radio tower guard spawner and unbind delegates.
  (Networking|SwitchHasAuthority
    (:Authority
      (Utilities|IsValid GuardSpawner
        (:Is%20Valid
          (if _returnvalue
            (Utilities|Map|Remove _radiotowerguardspawnersmap SpawnerIdTag)
            (Default|UnbindEventfromOnGuardsSpawned GuardSpawner (EventDispatchers|CreateEvent :!event HandleRadioTowerGuardsSpawned))
            (Fortnite|FortShippingLog self (Utilities|String|Append "UnregisterRadioTowerGuardSpawner: Unregistering bunker guard spawner: " (Utilities|GetDisplayName GuardSpawner)))
            (else
              (Fortnite|FortShippingLog self (Utilities|String|Append "UnregisterRadioTowerGuardSpawner: Spawner NOT already registered: " (Utilities|GetDisplayName GuardSpawner)) true))))))))

(fn UpdateRadioTowerActivationProgress
  (bind _returnvalue (BattleRoyaleGamePhaseLogic|GetBattleRoyaleGamePhaseLogicComponent))
  ;[0.1797 0.03639 0.1875] UpdateRadioTowerActivationProgress: Update the radio tower activation progress over time.
  (Networking|SwitchHasAuthority
    (:Authority
      (Utilities|FlowControl|Sequence
        (:then_0
          ;;[0.01776 0.278 0.3021] Cache how long it has been since the activation started.
          (LocalVariables|SetTimeSinceActivationStarted (- (Fortnite|Sync|Time|GetServerWorldTimeSeconds) (Variables|Transient|Activation|GetObjectiveActivateServerWorldTime)))
          ;;[0.02785 0.2612 0.3073] Cache the next SafeZonePhaseIndex and SafeZoneState.
          (bind _safezoneindicator (Variables|Default|GetSafeZoneIndicator)
            (:then
              (bind (_outsafezonephaseindex _outsafezonestate) (Helper|IncrementGamePhaseState (SafeZone|GetCurrentPhaseIndex _safezoneindicator) (SafeZone|GetSafeZoneState _safezoneindicator)))
              (LocalVariables|SetSafeZonePhaseIndex _outsafezonephaseindex)
              (LocalVariables|SetSafeZoneState _outsafezonestate)
              ;;[0.02615 0.009237 0.1406] Loop through all the future states to find the amount of time (in seconds) it is going to take to get there.
              (while (Helper|CompareGamePhaseState (Variables|Default|GetActivateComepleteSafeZoneIndex) (Utilities|Enum|BytetoEnumEFortSafeZoneState (Math|Conversions|ToByte(Integer) (ScalableFloat|GetValueasInteger (Variables|Hotfix|GetHotfix_ActivateCompleteSafeZoneState)))) (LocalVariables|GetSafeZonePhaseIndex) (LocalVariables|GetSafeZoneState))
                ;;;[1 0.5885 0.1497] Tally up the time left in the various future states.
                (bind _outxy (DataTable|EvaluateCurveTableRow "/Game/Athena/Balance/DataTables/AthenaGameData.AthenaGameData" (select (Utilities|Enum|Equal(Enum) (LocalVariables|GetSafeZoneState) "Holding") "Default.SafeZone.WaitTime" "Default.SafeZone.ShrinkTime") (Math|Conversions|ToFloat(Integer) (LocalVariables|GetSafeZonePhaseIndex)))
                  (:RowFound
                    (LocalVariables|SetFutureSafeZonesTimeLeft (+ (LocalVariables|GetFutureSafeZonesTimeLeft) _outxy))
                    ;;;[0.03069 0.1927 0.04119] Increment the variables for the next loop.
                    (bind (_outsafezonephaseindex_2 _outsafezonestate_2) (Helper|IncrementGamePhaseState (LocalVariables|GetSafeZonePhaseIndex) (LocalVariables|GetSafeZoneState)))
                    (LocalVariables|SetSafeZonePhaseIndex _outsafezonephaseindex_2)
                    (LocalVariables|SetSafeZoneState _outsafezonestate_2)))))
            (:else
              ;;[1 0 0] Warning: SafeZoneIndicator is null.
              (Fortnite|FortShippingLog self "UpdateRadioTowerActivationProgress: SafeZoneIndicator is NULL." true))))
        (:then_1
          ;;[0.03419 0.08364 0.1771] Update the RadioTowerActivationProgress in a normalized ratio.
          (Utilities|IsValid _returnvalue
            (:Is%20Valid
              (bind _output_get (Variables|Transient|Activation|SetRadioTowerActivationProgress (Math|Float|SafeDivide (LocalVariables|GetTimeSinceActivationStarted) (Utilities|Operators|Add (select (Variables|Transient|GetbActivateCompleteSafeZoneStateAndPhaseReached) (Utilities|Time|GetTimerRemainingTimebyHandle (Variables|Transient|GetActivateCompleteDelayTimerHandle)) (+ (BattleRoyaleGamePhaseLogic|GetGamePhaseStepTimeRemaining _returnvalue) (Variables|Default|GetActivateCompleteDelay))) (LocalVariables|GetTimeSinceActivationStarted) (LocalVariables|GetFutureSafeZonesTimeLeft)))))
              (Development|PrintString (Utilities|String|Append "RadioTowerActivationProgress: " (Utilities|String|ToString(Float) _output_get)) true false "(R=0.000000,G=0.660000,B=1.000000,A=1.000000)" 2.0 "RadioTower"))))))))
