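; Blueprint for a keycard pickup object in the BeetCalf gameplay content, likely a child of a shared pickup class.
; The construction script runs the parent's construction logic, then applies the KeycardMaterial variable to slot 0 of the PickupMesh, so each keycard's look follows that material setting.
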(fn ConstructionScript
  (|CallParentFunction)
  (Rendering|Material|SetMaterial (Variables|Default|GetPickupMesh) 0 (Variables|Default|GetKeycardMaterial)))
